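; Gas pump component that registers and unregisters its pump with the Valet gas pump indicator manager, likely for pump icons.
; On BeginPlay, the server in the Athena subgame checks an icon-enabled curve row in AthenaGameData, finds the manager component on the GameState, caches it and adds itself.
; On EndPlay, the server removes itself from the cached manager if it is valid.
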(event EventBeginPlay
  (bind _returnvalue (Fortnite|Athena|GameState|GetGameStateAthena))
  (bind _gas_pump_indicatormanager_class (Variables|Default|GetGas%20Pump%20IndicatorManager%20Class))
  (bind _returnvalue_2 (Actor|GetComponentbyClass _returnvalue _gas_pump_indicatormanager_class))
  ;[1 0 1] Begin Play
  (|CallParentFunction)
  (Utilities|SwitchHasAuthority
    (:Authority
      (if (Utilities|Enum|Equal(Enum) (Fortnite|GetSubGame) "Athena")
        (bind _outxy (DataTable|EvaluateCurveTableRow "/Game/Athena/Balance/DataTables/AthenaGameData.AthenaGameData" (Variables|Default|GetGas%20Pump%20Icon%20Enabled))
          (:RowFound
            (if (> _outxy 0.0)
              _))
          (:RowNotFound
            _))
        ;;[1 0.5 1] Add pump to gas pump indicator manager
        (Utilities|IsValid _returnvalue_2
          (:Is%20Valid
            (bind _asvalet_gas_pump_indicator_manager_v2 (Utilities|Casting|CastToValet_GasPumpIndicatorManagerV2 _returnvalue_2))
            (bind _output_get (Variables|Default|SetGas%20Pump%20IndicatorManager _asvalet_gas_pump_indicator_manager_v2))
            (Class|ValetGasPumpIndicatorManagerV2|AddGasPump _output_get self)))))))

(event EventEndPlay
  (params (EndPlayReason :type EEndPlayReason%20Enum :default "Destroyed"))
  ;[1 0 1] End Play
  (|CallParentFunction EndPlayReason)
  (Utilities|SwitchHasAuthority
    (:Authority
      (bind _gas_pump_indicatormanager (Variables|Default|GetGas%20Pump%20IndicatorManager)
        (:then
          (Class|ValetGasPumpIndicatorManagerV2|RemoveGasPump _gas_pump_indicatormanager self))
        (:else)))))
