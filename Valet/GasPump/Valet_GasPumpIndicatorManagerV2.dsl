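; Valet gas pump indicator manager that keeps fuel indicator actors near registered gas pump components.
; AddGasPump records the pump and spawns an indicator at its owner's transform when none is within range, so nearby pumps share one.
; RemoveGasPump unregisters the pump, collects indicators with no remaining pump in range, then destroys them in a separate pass.
; Range checks use squared 2D distance, ignoring height.
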(fn AddGasPump
  (params (Gas%20Pump :type B%20Gas%20Pump%20Valet%20Component%20Object%20Reference))
  (bind _returnvalue (Components|GetOwner Gas%20Pump))
  (Utilities|IsValid Gas%20Pump
    (:Is%20Valid
      (Utilities|Array|AddUnique (Variables|Default|GetGasPumpList) Gas%20Pump)
      ;[1 0 1] Find a nearby FuelIndicator
      (bind _array_element (Utilities|Array|ForEachLoopwithBreak (Variables|Default|GetFuelIndicatorList))
        (:LoopBody
          (Utilities|IsValid _array_element
            (:Is%20Valid
              (if (<= (Math|Vector|Distance2DSquared(Vector) (Transformation|GetActorLocation _array_element) (Transformation|GetActorLocation _returnvalue)) (Variables|Default|GetFuelIndicatorRangeSquared))
                (LocalVariables|SetFoundNearbyFuelIndicator true)))))
        (:Completed
          _))
      ;[0.75 0 0.75] If no nearby FuelIndicator actors found, add one.
      (if (not (LocalVariables|GetFoundNearbyFuelIndicator))
        (bind _returnvalue_2 (Game|SpawnActorfromClass (Variables|Default|GetFuelIndicatorClass) (Transformation|GetActorTransform _returnvalue) "Undefined" "SelectDefaultAtRuntime" 0 0))
        (Utilities|IsValid _returnvalue_2
          (:Is%20Valid
            (bind _asfuel_indicator (Utilities|Casting|CastToFuelIndicator _returnvalue_2))
            (Utilities|Array|Add (Variables|Default|GetFuelIndicatorList) _asfuel_indicator)))))))

(fn RemoveGasPump
  (params (Gas%20Pump :type B%20Gas%20Pump%20Valet%20Component%20Object%20Reference))
  (Utilities|IsValid Gas%20Pump
    (:Is%20Valid
      (Utilities|Array|RemoveItem (Variables|Default|GetGasPumpList) Gas%20Pump)
      (for _array_element (Variables|Default|GetFuelIndicatorList)
        ;[1 0 1] If no nearby Gas Pumps, mark FuelIndicator for removal and deletion.
        (Utilities|IsValid _array_element
          (:Is%20Valid
            (LocalVariables|SetFoundNearbyGasPump false)
            (bind _array_element_2 (Utilities|Array|ForEachLoopwithBreak (Variables|Default|GetGasPumpList))
              (:LoopBody
                (Utilities|IsValid (Components|GetOwner _array_element_2)
                  (:Is%20Valid
                    (if (<= (Math|Vector|Distance2DSquared(Vector) (Transformation|GetActorLocation (Components|GetOwner _array_element_2)) (Transformation|GetActorLocation _array_element)) (Variables|Default|GetFuelIndicatorRangeSquared))
                      (LocalVariables|SetFoundNearbyGasPump true)))))
              (:Completed
                _))
            (if (not (LocalVariables|GetFoundNearbyGasPump))
              (Utilities|Array|Add (LocalVariables|GetRemovalList) _array_element)))))
      (for _array_element_3 (LocalVariables|GetRemovalList)
        (Utilities|Array|RemoveItem (Variables|Default|GetFuelIndicatorList) _array_element_3)
        (Actor|DestroyActor _array_element_3)))))
